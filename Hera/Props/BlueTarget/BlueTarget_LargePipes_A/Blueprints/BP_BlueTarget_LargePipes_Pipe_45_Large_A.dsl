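; Blueprint prop for a large industrial 45-degree pipe piece from the BlueTarget LargePipes set, with optional traversable collision.
; Its construction script checks the UseTraversableCollision flag and, when it is false, destroys the SM_Industrial_LargePipes_Pipe_45_Large_A_CustomCollision component.
; When the flag is true, the script leaves that custom collision component in place.
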(fn ConstructionScript
  (if (not (Variables|Default|GetUseTraversableCollision))
    (Components|DestroyComponent (Variables|Default|GetSM_Industrial_LargePipes_Pipe_45_Large_A_CustomCollision))))
